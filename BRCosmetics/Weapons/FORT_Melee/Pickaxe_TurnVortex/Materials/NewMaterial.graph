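# Default Lit material for the TurnVortex pickaxe cosmetic, set up for skeletal meshes and mesh deformer use.
# It samples the pickaxe's diffuse texture with texture coordinates divided by a scalar parameter, which acts as a tiling control and defaults to no change.
# A lone constant accompanies the sample, and the generic NewMaterial name and minimal graph have no other shading logic.
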Material /BRCosmetics/Weapons/FORT_Melee/Pickaxe_TurnVortex/Materials/NewMaterial
graph {
 properties {bAutomaticallySetUsageInEditor=false, bUsedWithMeshDeformer=true, bUsedWithSkeletalMesh=true, usedShadingModels=Default Lit}
 nodes [
  Constant()
  TextureSample(
   UVs=Divide(A=TextureCoordinate(), B=ScalarParameter() {defaultValue=1})
   Tex=TextureObject() {texture=/BRCosmetics/Weapons/FORT_Melee/Pickaxe_TurnVortex/Textures/T_TurnVortex_Pickaxe_D}
  ) {texture=&T_TurnVortex_Pickaxe_D}
 ]
}
